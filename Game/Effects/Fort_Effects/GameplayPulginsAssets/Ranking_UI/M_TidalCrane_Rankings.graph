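# Additive unlit Niagara sprite material for ranking UI effects that shows one frame of a placeholder ranking mask texture.
# A SubUV function slices the texture into five horizontal frames, selected by the Frame scalar parameter.
# The frame's green channel multiplies particle color into a shared FX emissive adjustment function, and separately multiplies particle alpha.
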Material /Game/Effects/Fort_Effects/GameplayPulginsAssets/Ranking_UI/M_TidalCrane_Rankings
graph {
 properties {
  bAutomaticallySetUsageInEditor=false
  bUsedWithNiagaraSprites=true
  blendMode=BLEND_Additive
  shadingModel=MSM_Unlit
  usedShadingModels=Unlit
 }
 nodes [
  n1 = ParticleColor()
  n2 = MaterialFunctionCall(Float3=MaterialFunctionCall!(
    Texture=TextureObject() {samplerType=SAMPLERTYPE_Masks, texture=&/T_Ranking_PlaceHolder}
    UVs
    SubImages=Constant2Vector() {g=1, r=5}
    Frame=ScalarParameter() {parameterName=Frame}
   ) {materialFunction=/Engine/Functions/Engine_MaterialFunctions03/Texturing/SubUV_Function}.RGB) {materialFunction=&BreakOutFloat3Components}
  MaterialFunctionCall(Emissive=Multiply(A=@n1.RGB, B=@n2.G)) {materialFunction=&MF_FX_EmissiveAdjustments}
  Multiply(A=@n1.A, B=@n2.G)
 ]
}
